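# M_Mustard is a material in the FX folder of the MustardRocket player emote, presumably shading the mustard effect.
# It uses Default Lit shading and is flagged for Niagara ribbons, with automatic usage flagging in the editor turned off.
# Its graph holds a single vector parameter defaulting to mustard yellow, with no textures, math or visible output wiring.
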Material /BRCosmetics/Animation/Game/MainPlayer/Emotes/MustardRocket/FX/M_Mustard
graph {
 properties {bAutomaticallySetUsageInEditor=false, bUsedWithNiagaraRibbons=true, usedShadingModels=Default Lit}
 nodes [VectorParameter() {defaultValue=<rgba 0.7656 0.4042 0.0524 1>}]
}
