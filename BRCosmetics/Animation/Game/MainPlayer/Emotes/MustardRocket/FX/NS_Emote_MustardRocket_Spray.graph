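# Niagara system for the MustardRocket emote that sprays pulsing mustard from a nozzle on a skeletal mesh.
# Looping squeeze curves pulse the muzzle mesh, ribbon trail, liquid sprites and ground decals.
# The ribbon bends from nozzle toward a collision-traced landing point, while decals spawn at wave-animated offsets from the player pawn.
# Smoothed owner-speed thresholds reshape the spray during movement, and decal colors vary by time of day or front-end preview.
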NiagaraSystem /BRCosmetics/Animation/Game/MainPlayer/Emotes/MustardRocket/FX/NS_Emote_MustardRocket_Spray
dependencies {
 usedDataInterfaces [
  &NiagaraDataInterfaceCurve
  &NiagaraDataInterfaceCollisionQuery
  &NiagaraDataInterfaceDebugDraw
  &NiagaraDataInterfaceParticleRead
 ]
 usedDynamicInputs [&Multiply_VectorByFloat]
 usedModules [
  &self:SystemSpawnScript.SystemScriptSource.SystemScriptGraph.NiagaraNodeAssignment_3.SetVariables_#1
  &SystemState
  &self:SystemSpawnScript.SystemScriptSource.SystemScriptGraph.NiagaraNodeAssignment_0.SetVariables_#2
  &self:SystemSpawnScript.SystemScriptSource.SystemScriptGraph.NiagaraNodeAssignment_1.SetVariables_#3
  &self:SystemSpawnScript.SystemScriptSource.SystemScriptGraph.NiagaraNodeAssignment_2.SetVariables_#4
  &InterpolateOverTime
  &self:SystemSpawnScript.SystemScriptSource.SystemScriptGraph.NiagaraNodeAssignment_4.SetVariables_#5
  &self:SystemSpawnScript.SystemScriptSource.SystemScriptGraph.NiagaraNodeAssignment_5.SetVariables_#6
 ]
}
emitters {
 MustardDecals {
  emitterUpdateScript [
   &EmitterState(
    Inactive Response=::NewEnumerator1
    Life Cycle Mode=::NewEnumerator1
    Loop Duration=9.5
    Recalculate Duration Each Loop=true
    UseLoopDelay=false
   )
   /SetVariables(Emitter.Squeeze:NiagaraFloat=&FloatFromCurve(
     FloatCurve=DataInterface() {propertyValues {Curve {keys [<at 0.12 0>,<at 0.3975 1>,<at 0.7795 0>,<at 1.2227 0>,<at 1.4039 1>,<at 1.9625 0>,<at 2.3888 0>,<at 2.5167 1>,<at 3.3183 0>,<at 3.5941 0>,<at 3.6814 0.95>,<at 4.3747 0>,<at 4.8492 0>,<at 4.9668 0.95>,<at 5.5707 0>,<at 5.87 0>,<at 6.0408 1>,<at 6.5746 0>,<at 6.96 0>,<at 7.271 1>,<at 8.0606 0>,<at 8.22 0>,<at 8.4299 1>,<at 9.2089 0>]}}}
     CurveIndex=$Emitter.LoopedAge
    )) {isSetParameters=true}
   &SpawnRate(SpawnRate=&Lerp_Float(A=6, B=70, Alpha=$Emitter.Squeeze), Use Spawn Probability=false)
   &self:InterpolateSpeedThresholdEmitter_05(
    Define Initial Value=false
    Rate Of Change=7
    Target Float Value:NiagaraFloat=&MakeCustomFloatFromBool(
     Bool=&SetBoolByFloatComparison(
      A=&Add_Float(
       A=&Add_Float(
        A=&Abs_Float(Float:NiagaraFloat=&MakeFloatFromVector(Vector=$Engine.Owner.Velocity, Channel=::NewEnumerator0))
        B=&Abs_Float(Float:NiagaraFloat=&MakeFloatFromVector(Vector=$Engine.Owner.Velocity, Channel=::NewEnumerator2))
       )
       B=&Abs_Float(Float:NiagaraFloat=&MakeFloatFromVector(Vector=$Engine.Owner.Velocity, Channel=::NewEnumerator1))
      )
      B=400
     )
     True Float=1
     False Float=0
    )
   ) {module=InterpolateSpeedThresholdEmitter_05}
   /SetVariables(Emitter.Linear Color:LinearColor=&FNTimeOfDayColor(
     Night Color=<rgba 1 0.2689 0.3071 1>
     Morning Color=<rgba 1 0.2423 0.5265 1>
     Day Color=<rgba 1 0.4891 0.9128 1>
     Evening Color=<rgba 1 0.2979 0.4227 1>
    )) {isSetParameters=true}
   /SetVariables(
    Emitter.MustardColA:LinearColor=&MakeCustomLinearColorFromBool(
     Bool=$User.bIsFrontEndPreview
     True Color=<rgba 0.2917 0.1483 0.0005 1>
     False Color=&FNTimeOfDayColor(
      Night Color=<rgba 0.6111 0.4198 0.0637 0.3>
      Morning Color=<rgba 0.6111 0.525 0.0022 0.3>
      Day Color=<rgba 1 0.5816 0 0.3>
      Evening Color=<rgba 0.6111 0.525 0.0022 0.3>
     )
    )
    Emitter.MustardColB:LinearColor=&MakeCustomLinearColorFromBool(Bool=$User.bIsFrontEndPreview, True Color=<rgba 0.0365 0.0144 0 1>, False Color=<rgba 0.6783 0.3217 0.0133 0.1>)
   ) {isSetParameters=true}
  ]
  particleSpawnScript [
   &InitializeParticle(
    Color=&MakeCustomLinearColorFromBool(Bool=$User.bIsFrontEndPreview, True Color=<rgba 2 1.2162 0 1>, False Color=<rgba 1 1 1 1>)
    Lifetime Max=1
    Lifetime Min=0.7
    Lifetime Mode=::NewEnumerator1
    Mass Max=1.5
    Mass Mode=::NewEnumerator3
    Sprite Rotation Mode=::NewEnumerator2
    Sprite Size Mode=::NewEnumerator1
    Uniform Sprite Size Max=8
    Uniform Sprite Size Min=4
    UsePositionOffset=false
    Write Material Random=true
   )
   &SkeletalMeshLocation(
    Bone Sampling Mode=::NewEnumerator4
    MaskSpawnGroup=false
    Orientation Sampling=::NewEnumerator1
    Sampled Position Offset=&MakeVector(
     X=&Waveform(
      Global Phase=$Engine.System.Age
      "Waveform [1]"=::NewEnumerator0
      "[1] Phase"=$Engine.Time
      "[1] Frequency"=2
      "[1] Amplitude Min/Max"=<-20 20>
      "[2] Average"=false
      "[2] Input Position"=0
      "[2] Phase"=$Emitter.Age
      "[2] Phase Scale"=1
      "[2] Frequency"=1
      "[2] Amplitude Min/Max"=<-10 10>
      "[2] Scale"=1
      "[2] Bias"=0
      "[2] Exponent"=1
     )
     Y=&Add_Float(
      A=&Lerp_Float(
       A=&RandomRangeFloat(Minimum=-70, Maximum=-120)
       B=&RandomRangeFloat(Minimum=-70, Maximum=-90)
       Alpha=$Emitter.InterpolateSpeedThresholdEmitter_05.InterpolateOverTime_Float.Moving Average
      )
      B=&Waveform(
       Waveform Count=::NewEnumerator1
       Global Input Position=0.12
       Global Phase=$Engine.System.Age
       "Waveform [1]"=::NewEnumerator0
       "[1] Phase"=$Engine.Time
       "[1] Frequency"=2
       "[1] Amplitude Min/Max"=<-10 10>
       "[2] Average"=false
       "[2] Input Position"=0
       "[2] Phase"=$Emitter.Age
       "[2] Phase Scale"=1
       "[2] Compound Frequency"=2
       "[2] Frequency"=1
       "[2] Amplitude Min/Max"=<-10 10>
       "[2] Scale"=1
       "[2] Bias"=0
       "[2] Exponent"=1
       "[2] Pulse Duration"=0.1
       "[2] Clip Triangle Min/Max"=<0 0>
       "[2] Random Seed"=0
      )
     )
     Z=0
    )
    Sampled Position Offset Coordinate Space:ENiagaraSkelMeshTransforms=::NewEnumerator2
    Skeletal Mesh=$User.PlayerPawn
   )
   &DynamicMaterialParameters(
    Index 0 Param 1=2
    Module.Index0Float0Write=true
    Module.Index0Float1Write=true
    Module.Index0Float2Write=true
    Module.Index0Float3Write=true
   )
   /SetVariables(
    Particles.Rand:NiagaraFloat=&RandomRangeFloat(Minimum=8, Maximum=19)
    Particles.VisibilityTag=&RandomRangeInt(Minimum=0, Maximum=3)
   ) {isSetParameters=true}
  ]
  particleUpdateScript [
   &ParticleState()
   /SetVariables(Particles.DecalSize:Vector3f=&Multiply_Vector(
     A=&MakeVector(X=100, Y=$Particles.Rand, Z=$Particles.Rand)
     B=&VectorFromFloat(Value=&FloatFromCurve(FloatCurve=DataInterface() {propertyValues {Curve {keys [
           {arriveTangent=4.1785, interpMode=RCIM_Cubic, leaveTangent=4.1785, tangentMode=RCTM_User, time=0, value=0.5}
           {arriveTangent=1.1179, interpMode=RCIM_Cubic, leaveTangent=1.1179, tangentMode=RCTM_User, time=0.01, value=0.8}
           {arriveTangent=0.0224, interpMode=RCIM_Cubic, leaveTangent=0.0224, tangentMode=RCTM_User, time=0.05, value=1}
          ]}}}))
    )) {isSetParameters=true}
   /SetVariables(Particles.DecalFade:NiagaraFloat=&FloatFromCurve(FloatCurve=DataInterface() {propertyValues {Curve {keys [<at 0 1>,<at 1 0>]}}})) {isSetParameters=true}
   &ScaleColor(
    Scale Alpha=&FloatFromCurve(FloatCurve=DataInterface() {propertyValues {Curve {keys [<at 0 1>,<at 1 0>]}}})
    ScaleA=true
    ScaleRGB=true
   )
   /Niagara/Modules/AttributeReader/UpdateParticlesFromOtherEmitter(
    Attribute Reader=DataInterface() {propertyValues {EmitterBinding {bindingMode=Other, emitterName=OrbitBlobs_Trails002}}}
    Dynamic Material Parameter 1 Sampling:ENiagara_DirectReadSamplingMode=::NewEnumerator1
    Position Sampling:ENiagara_DirectReadSamplingMode=::NewEnumerator1
    Transform Data Option=1
   ) {enabled=false}
   /SetVariables(Particles.ConstructPos:Vector3f=&MakeVector(
     X=&MakeFloatFromVector4(Vector 4=$Output.UpdateParticlesFromOtherEmitter.DynamicMaterialParameter1, Channel=::NewEnumerator0)
     Y=&MakeFloatFromVector4(Vector 4=$Output.UpdateParticlesFromOtherEmitter.DynamicMaterialParameter1, Channel=::NewEnumerator1)
     Z=&MakeFloatFromVector4(Vector 4=$Output.UpdateParticlesFromOtherEmitter.DynamicMaterialParameter1, Channel=::NewEnumerator2)
    )) {enabled=false, isSetParameters=true}
   /SetVariables(Particles.Position=$Particles.SkeletalMeshLocation.SampledPosition) {enabled=false, isSetParameters=true}
  ]
  rendererClasses [&NiagaraDecalRendererProperties]
  renderers [
   {rendererClass=&NiagaraDecalRendererProperties, rendererIndex=0}
   {rendererClass=&NiagaraDecalRendererProperties, rendererIndex=1}
   {rendererClass=&NiagaraDecalRendererProperties, rendererIndex=2}
   {rendererClass=&NiagaraDecalRendererProperties, rendererIndex=3}
  ]
 }
 MustardLiquid {
  emitterUpdateScript [
   &EmitterState(
    Life Cycle Mode=::NewEnumerator1
    Loop Delay=&UniformRangedFloat(Minimum=0.8, Maximum=1.5)
    Loop Duration=9.5
    UseLoopDelay=false
   )
   /SetVariables(Emitter.Squeeze:NiagaraFloat=&FloatFromCurve(
     FloatCurve=DataInterface() {propertyValues {Curve {keys [<at 0 0>,<at 0.2775 1>,<at 0.551 0>,<at 1.1027 0>,<at 1.2839 1>,<at 1.6886 0>,<at 2.2688 0>,<at 2.3967 1>,<at 2.9 0>,<at 3.4741 0>,<at 3.5614 0.95>,<at 3.9 0>,<at 4.7292 0>,<at 4.8468 0.95>,<at 5.1784 0>,<at 5.75 0>,<at 5.9208 1>,<at 6.0989 0>,<at 6.84 0>,<at 7.151 1>,<at 7.4261 0>,<at 8.1 0>,<at 8.3099 1>,<at 9.0889 0>]}}}
     CurveIndex=$Emitter.LoopedAge
    )) {isSetParameters=true}
   &SpawnRate(SpawnRate=&Lerp_Float(A=6, B=37, Alpha=$Emitter.Squeeze), Use Spawn Probability=false) {module=SpawnRate001}
   &self:InterpolateSpeedThresholdEmitter_05(
    Define Initial Value=false
    Rate Of Change=7
    Target Float Value:NiagaraFloat=&MakeCustomFloatFromBool(
     Bool=&SetBoolByFloatComparison(
      A=&Add_Float(
       A=&Add_Float(
        A=&Abs_Float(Float:NiagaraFloat=&MakeFloatFromVector(Vector=$Engine.Owner.Velocity, Channel=::NewEnumerator0))
        B=&Abs_Float(Float:NiagaraFloat=&MakeFloatFromVector(Vector=$Engine.Owner.Velocity, Channel=::NewEnumerator2))
       )
       B=&Abs_Float(Float:NiagaraFloat=&MakeFloatFromVector(Vector=$Engine.Owner.Velocity, Channel=::NewEnumerator1))
      )
      B=400
     )
     True Float=1
     False Float=0
    )
   ) {module=InterpolateSpeedThresholdEmitter_05}
  ]
  particleSpawnScript [
   /Niagara/Modules/Spawn/Initialization/InitializeParticle(
    Color=<rgba 2 2 2 1>
    Lifetime=&RandomRangeFloat(Minimum=0.7, Maximum=1.2)
    Mesh Scale=<10 10 10>
    Module.Write Color=true
    Module.Write Lifetime=true
    Module.Write Mass=false
    Module.Write Material Random=true
    Module.Write Position=true
    Module.Write Scale=true
    Module.Write SpriteRotation=true
    Module.Write SpriteSize=true
    Sprite Rotation=&RandomRangeFloat(Minimum=-180, Maximum=180)
    Sprite Size=&Multiply_Vector2D_ByFloat(Float=&RandomRangeFloat(Minimum=0.4, Maximum=1), Vector2D=<55 55>)
   )
   /SetVariables(Particles.UVScale:Vector2f=&MakeVector2D(X=&UniformAOrBFloat(A=-1), Y=1)) {isSetParameters=true}
   &DynamicMaterialParameters(
    Index 0 Param 1=&RandomRangeFloat(Minimum=0.1, Maximum=1)
    Index 0 Param 2=&RandomRangeFloat(Minimum=0.2, Maximum=0.5)
    Index 0 Param 3=&RandomRangeFloat(Minimum=50, Maximum=100)
    Index 0 Param 4=&RandomRangeFloat(Minimum=0, Maximum=10)
    Module.Param0WriteEnabled=true
    Module.Param1WriteEnabled=true
    Module.Param2WriteEnabled=true
    Module.Param3WriteEnabled=true
   ) {module=DynamicMaterialParameters001}
   &AddVelocity(Velocity=&Lerp_Vector(
     A=&RandomRangeVector(Minimum=<0 -50 -50>, Maximum=<0 -175 -150>)
     B=&Lerp_Vector(
      A=&RandomRangeVector(Minimum=<0 -50 160>, Maximum=<0 -175 -10>)
      B=&RandomRangeVector(Minimum=<0 -70 50>, Maximum=<0 40 -100>)
      Alpha=$Emitter.InterpolateSpeedThresholdEmitter_05.InterpolateOverTime_Float.Moving Average
     )
     Alpha=$Emitter.Squeeze
    ))
   &Color(
    Color=<rgba 0.7652 0.3252 0 1>
    Scale Alpha=&FloatFromCurve(FloatCurve=DataInterface() {propertyValues {Curve {keys [
         {interpMode=RCIM_Cubic, tangentMode=RCTM_Break, time=0, value=1}
         {arriveTangent=-3.5002, interpMode=RCIM_Cubic, tangentMode=RCTM_Break, time=1, value=0}
        ]}}})
   ) {module=Color001}
   &SkeletalMeshLocation(
    Bone Sampling Mode=::NewEnumerator4
    MaskSpawnGroup=false
    Orientation Sampling=::NewEnumerator1
    Position Sampling=::NewEnumerator2
    Sampled Position Offset=&MakeVector(
     X=&Waveform(
      Global Phase=$Engine.System.Age
      "Waveform [1]"=::NewEnumerator0
      "[1] Phase"=$Engine.Time
      "[1] Frequency"=2
      "[1] Amplitude Min/Max"=<-10 10>
      "[2] Average"=false
      "[2] Input Position"=0
      "[2] Phase"=$Emitter.Age
      "[2] Phase Scale"=1
      "[2] Frequency"=1
      "[2] Amplitude Min/Max"=<-10 10>
      "[2] Scale"=1
      "[2] Bias"=0
      "[2] Exponent"=1
     )
     Y=&Add_Float(
      A=&RandomRangeFloat(Minimum=-85, Maximum=-125)
      B=&Waveform(
       Waveform Count=::NewEnumerator1
       Global Input Position=0.12
       Global Phase=$Engine.System.Age
       "Waveform [1]"=::NewEnumerator0
       "[1] Phase"=$Engine.Time
       "[1] Frequency"=2
       "[1] Amplitude Min/Max"=<-10 10>
       "[2] Average"=false
       "[2] Input Position"=0
       "[2] Phase"=$Emitter.Age
       "[2] Phase Scale"=1
       "[2] Compound Frequency"=2
       "[2] Frequency"=1
       "[2] Amplitude Min/Max"=<-10 10>
       "[2] Scale"=1
       "[2] Bias"=0
       "[2] Exponent"=1
       "[2] Pulse Duration"=0.1
       "[2] Clip Triangle Min/Max"=<0 0>
       "[2] Random Seed"=0
      )
     )
     Z=0
    )
    Sampled Position Offset Coordinate Space:ENiagaraSkelMeshTransforms=::NewEnumerator2
    Skeletal Mesh=$User.PlayerPawn
   ) {enabled=false, module=SkeletalMeshLocation002}
   &SkeletalMeshLocation(Bone Sampling Mode=::NewEnumerator4, MaskSpawnGroup=false, Skeletal Mesh=$System.Skeletal Mesh) {module=SkeletalMeshLocation001}
  ]
  particleUpdateScript [
   &ParticleState()
   &GravityForce(Gravity=<0 0 -800>)
   &Drag(Drag=3, Use Linear Drag=true, Use Rotational Drag=false) {enabled=false}
   &InheritVelocity(Inherited Velocity Amount Scale=<0.05 0.05 0.05>, Speed Limit=true, UseSpeedThreshold=false)
   &SolveForcesAndVelocity(Clamp Velocity=false, Limit Acceleration=false)
   &ScaleSpriteSizeBySpeed(Max Scale Factor=<1 1.5>, Velocity Threshold=50)
   &ScaleSpriteSize(Scale Factor=&Vector2DFromFloat(Value=&FloatFromCurve(
      FloatCurve=DataInterface() {propertyValues {Curve {keys [
          {arriveTangent=5.9795, interpMode=RCIM_Cubic, leaveTangent=5.9795, tangentMode=RCTM_User, time=0, value=0.2}
          {arriveTangent=0.0022, interpMode=RCIM_Cubic, leaveTangent=0.0022, tangentMode=RCTM_User, time=0.7, value=1}
          {arriveTangent=-7.4026, interpMode=RCIM_Cubic, leaveTangent=-7.4026, tangentMode=RCTM_User, time=1, value=0}
         ]}}}
      Scale Curve=0.6
     ))) {module=ScaleSpriteSize001}
   &ScaleColor(
    Scale Alpha=&FloatFromCurve(FloatCurve=DataInterface() {propertyValues {Curve {keys [
         {arriveTangent=-2.5249, interpMode=RCIM_Cubic, time=0.1, value=1}
         {arriveTangent=-0.9941, interpMode=RCIM_Cubic, leaveTangent=-0.9941, tangentMode=RCTM_User, time=0.3843, value=0.4072}
         {arriveTangent=-1.0623, interpMode=RCIM_Cubic, leaveTangent=-1.0623, tangentMode=RCTM_User, time=1, value=0}
        ]}}})
    Scale RGB=&MakeVectorFromLinearColorRGB(Color=<rgba 1 1 1 1>)
    ScaleA=true
    ScaleRGB=false
   )
   &SpriteRotationRate(Rotation Rate=10) {enabled=false}
   &DynamicMaterialParameters(
    Index 1 Param 1=&FloatFromCurve(FloatCurve=DataInterface() {propertyValues {Curve {keys [
         {arriveTangent=4.2071, interpMode=RCIM_Cubic, time=0, value=-2}
         {arriveTangent=1.1564, interpMode=RCIM_Cubic, leaveTangent=1.1564, tangentMode=RCTM_User, time=1, value=1}
        ]}}})
    Module.Param0WriteEnabled1=true
    Module.Param1WriteEnabled1=false
    Module.Param2WriteEnabled1=false
    Module.Param3WriteEnabled1=false
    Write Parameter Index 0=false
    Write Parameter Index 1=true
   )
   /SetVariables(Particles.Position=&Lerp_Vector(
     A=$Particles.Initial.SkeletalMeshLocation001.SampledPosition
     B=$Particles.Initial.SkeletalMeshLocation002.SampledPosition
     Alpha=&FloatFromCurve(FloatCurve=DataInterface() {propertyValues {Curve {
         defaultValue=1
         keys [
          {arriveTangent=-0.3254, interpMode=RCIM_Cubic, leaveTangent=-0.3254, tangentMode=RCTM_User, time=0, value=0}
          {arriveTangent=4.3425, interpMode=RCIM_Cubic, leaveTangent=4.3425, tangentMode=RCTM_User, time=0.4, value=1}
         ]
        }}})
    )) {enabled=false, isSetParameters=true}
  ]
  rendererClasses [&NiagaraSpriteRendererProperties]
  renderers [{rendererClass=&NiagaraSpriteRendererProperties}]
 }
 MustardMuzzle {
  emitterUpdateScript [
   &EmitterState(Inactive Response=::NewEnumerator1, Life Cycle Mode=::NewEnumerator1, Loop Duration=9.5, UseLoopDelay=false)
   /SetVariables(Emitter.Squeeze:NiagaraFloat=&FloatFromCurve(
     FloatCurve=DataInterface() {propertyValues {Curve {keys [<at 0 0>,<at 0.2775 1>,<at 0.551 0>,<at 1.1027 0>,<at 1.2839 1>,<at 1.6886 0>,<at 2.2688 0>,<at 2.3967 1>,<at 2.9 0>,<at 3.4741 0>,<at 3.5614 0.95>,<at 3.9 0>,<at 4.7292 0>,<at 4.8468 0.95>,<at 5.1784 0>,<at 5.75 0>,<at 5.9208 1>,<at 6.0989 0>,<at 6.84 0>,<at 7.151 1>,<at 7.4261 0>,<at 8.1 0>,<at 8.3099 1>,<at 9.0889 0>]}}}
     CurveIndex=$Emitter.LoopedAge
    )) {isSetParameters=true}
   /SetVariables(Emitter.SqueezeSize:NiagaraFloat=&Lerp_Float(
     A=1
     B=&FloatFromCurve(
      FloatCurve=DataInterface() {propertyValues {Curve {keys [<at 0 0>,<at 0.2 1>,<at 1 0>,<at 1.1027 0>,<at 1.2 1>,<at 2.1 0>,<at 2.2688 0>,<at 2.3 1>,<at 3.3 0>,<at 3.4741 0>,<at 3.5 1>,<at 4.3 0>,<at 4.7292 0>,<at 4.8 1>,<at 5.6 0>,<at 5.75 0>,<at 5.8 1>,<at 6.5 0>,<at 6.84 0>,<at 7.1 1>,<at 7.9 0>,<at 8.1 0>,<at 8.2 1>,<at 9.3 0>]}}}
      CurveIndex=$Emitter.LoopedAge
      Scale Curve=1.4
     )
     Alpha=0
    )) {isSetParameters=true}
   &SpawnRate(SpawnRate=&Lerp_Float(A=5, B=8, Alpha=$Emitter.Squeeze), Use Spawn Probability=false)
   /SetVariables(Emitter.float:NiagaraFloat=&FloatFromCurve(
     FloatCurve=DataInterface() {propertyValues {Curve {keys [{arriveTangent=-70.2419, interpMode=RCIM_Cubic, leaveTangent=-70.2419, tangentMode=RCTM_User, time=0, value=30}, <at 1 30 RCIM_Cubic>]}}}
     CurveIndex=$Emitter.NormalizedLoopAge
    )) {isSetParameters=true}
  ]
  particleSpawnScript [
   /Niagara/Modules/Spawn/Initialization/V2/InitializeParticle(
    Color=&FNTimeOfDayColor(
     Night Color=<rgba 1 0.485 0.0312 1>
     Morning Color=<rgba 1 0.3923 0.0399 1>
     Day Color=<rgba 1 0.3923 0.0399 1>
     Evening Color=<rgba 1 0.3923 0.0399 1>
    )
    Lifetime=&RandomRangeFloat(Minimum=0.4, Maximum=0.6)
    Mesh Scale=&Lerp_Vector(A=<0.35 0.25 0.35>, B=<0.38 0.45 0.38>, Alpha=$Emitter.Squeeze)
    Mesh Scale Mode=::NewEnumerator4
    UsePositionOffset=true
    Write Material Random=true
   )
   &InitialMeshOrientation(
    Mesh Orientation Mode=::NewEnumerator7
    Rotation=&MakeVector(X=0, Y=&RandomRangeFloat(Minimum=0, Maximum=1), Z=0)
    Use Rotation Vector=true
   ) {enabled=false}
  ]
  particleUpdateScript [
   &ParticleState()
   &self:ScratchModule_03(
    Define Initial Value=false
    Rate Of Change=10
    Target Float Value:NiagaraFloat=&MakeCustomFloatFromBool(
     Bool=&SetBoolByFloatComparison(
      A=&Add_Float(
       A=&Add_Float(
        A=&Abs_Float(Float:NiagaraFloat=&MakeFloatFromVector(Vector=$Engine.Owner.Velocity, Channel=::NewEnumerator0))
        B=&Abs_Float(Float:NiagaraFloat=&MakeFloatFromVector(Vector=$Engine.Owner.Velocity, Channel=::NewEnumerator2))
       )
       B=&Abs_Float(Float:NiagaraFloat=&MakeFloatFromVector(Vector=$Engine.Owner.Velocity, Channel=::NewEnumerator1))
      )
      B=700
     )
     True Float=1
     False Float=0
    )
   ) {module=ScratchModule_03}
   &ScaleMeshSize(Scale Factor=&MakeVector(
     X=&FloatFromCurve(
      FloatCurve=DataInterface() {propertyValues {Curve {
         defaultValue=1
         keys [
          {arriveTangent=3.1457, interpMode=RCIM_Cubic, leaveTangent=3.1457, tangentMode=RCTM_User, time=0, value=0}
          {arriveTangent=0.5, interpMode=RCIM_Cubic, leaveTangent=0.5, time=0.1, value=1}
          <at 1 0.5 RCIM_Cubic>
         ]
        }}}
      Scale Curve=&Lerp_Float(
       A=$Emitter.SqueezeSize
       B=0
       Alpha=&Abs_Float(Float:NiagaraFloat=$Particles.ScratchModule_03.InterpolateOverTime_Float.Moving Average)
      )
     )
     Y=&FloatFromCurve(
      FloatCurve=DataInterface() {propertyValues {Curve {
         defaultValue=1
         keys [
          {arriveTangent=3.1457, interpMode=RCIM_Cubic, leaveTangent=3.1457, tangentMode=RCTM_User, time=0, value=0}
          <at 0.5 3 RCIM_Cubic>
          <at 1 0 RCIM_Cubic>
         ]
        }}}
      Scale Curve=&Lerp_Float(
       A=$Emitter.SqueezeSize
       B=0
       Alpha=&Abs_Float(Float:NiagaraFloat=$Particles.ScratchModule_03.InterpolateOverTime_Float.Moving Average)
      )
     )
     Z=&FloatFromCurve(
      FloatCurve=DataInterface() {propertyValues {Curve {
         defaultValue=1
         keys [
          {arriveTangent=3.1457, interpMode=RCIM_Cubic, leaveTangent=3.1457, tangentMode=RCTM_User, time=0, value=0}
          {arriveTangent=0.5, interpMode=RCIM_Cubic, leaveTangent=0.5, time=0.1, value=1}
          <at 1 0.5 RCIM_Cubic>
         ]
        }}}
      Scale Curve=&Lerp_Float(
       A=$Emitter.SqueezeSize
       B=0
       Alpha=&Abs_Float(Float:NiagaraFloat=$Particles.ScratchModule_03.InterpolateOverTime_Float.Moving Average)
      )
     )
    ))
   &ScaleMeshSize(Scale Factor=&VectorFromFloat(Value=$Emitter.SqueezeSize)) {enabled=false, module=ScaleMeshSize001}
   &DynamicMaterialParameters(
    Index 0 Param 1=&Lerp_Float(
     A=0
     B=1
     Alpha=&FloatFromCurve(
      FloatCurve=DataInterface() {propertyValues {Curve {keys [
          {arriveTangent=5.87, interpMode=RCIM_Cubic, leaveTangent=5.87, tangentMode=RCTM_User, time=0, value=0}
          {interpMode=RCIM_Cubic, tangentMode=RCTM_User, time=0.5, value=1}
         ]}}}
      CurveIndex=$System.FN_Contrail_Setup.Dive Value
     )
    )
    Index 0 Param 2=&Lerp_Float(
     A=0
     B=10
     Alpha=&FloatFromCurve(
      FloatCurve=DataInterface() {propertyValues {Curve {keys [
          {arriveTangent=5.87, interpMode=RCIM_Cubic, leaveTangent=5.87, tangentMode=RCTM_User, time=0, value=0}
          {interpMode=RCIM_Cubic, tangentMode=RCTM_User, time=0.5, value=1}
         ]}}}
      CurveIndex=$System.FN_Contrail_Setup.Dive Value
     )
    )
    Index 0 Param 3=&RandomRangeFloat(Minimum=0, Maximum=0)
    Index 0 Param 4=&RandomRangeFloat(Minimum=0, Maximum=0)
    Module.Index0Param0Write=true
    Module.Index0Param1Write=true
    Module.Index0Param2Write=true
    Module.Index0Param3Write=true
   )
   &SampleSkeletalMesh(
    AdditionalYawPitchRoll=false
    Bone Sampling Mode:ENiagaraBoneSamplingMode=::NewEnumerator4
    Skeletal Mesh:NiagaraDataInterfaceSkeletalMesh=$System.Skeletal Mesh
   )
   &SkeletalMeshLocation(Bone Sampling Mode=::NewEnumerator4, MaskSpawnGroup=false, Orientation Sampling=::NewEnumerator1, Skeletal Mesh=$System.Skeletal Mesh)
   &self:ScratchModule_04(
    Define Initial Value=false
    Rate Of Change=8
    Target Float Value:NiagaraFloat=&RemapRange(
     Input Value To Remap=&MakeFloatFromVector(
      Vector=&ReturnMeshOrientationAxis(Mesh Orientation Axis=::ZAxis, Mesh Orientation Quat=$Particles.SampleSkeletalMesh.SampledRotation)
      Channel=::NewEnumerator2
     )
     Input Min=0.93
     Input Max=1
     Output Min=0
     Output Max=1
    )
   ) {module=ScratchModule_04}
   &UpdateMeshOrientation(Rotation Vector=&MakeVector(X=&Lerp_Float(A=0.4, B=-0.3, Alpha=$Emitter.Squeeze), Y=0, Z=0))
   &ScaleColor(
    Scale Alpha=&FloatFromCurve(FloatCurve=DataInterface() {propertyValues {Curve {keys [<at 0 1>,<at 1 0>]}}})
    ScaleA=true
    ScaleRGB=true
   )
   &MeshRotationForce(Rotation=&MakeVector(X=0, Y=&RandomRangeFloat(Minimum=-10, Maximum=10), Z=0))
   &SolveForcesAndVelocity(Clamp Velocity=false, Limit Acceleration=false)
  ]
  rendererClasses [&NiagaraMeshRendererProperties]
  renderers [{rendererClass=&NiagaraMeshRendererProperties, rendererIndex=0}]
 }
 MustardTrail {
  emitterUpdateScript [
   &EmitterState(
    Inactive Response=::NewEnumerator1
    Life Cycle Mode=::NewEnumerator1
    Loop Duration=9.5
    Recalculate Duration Each Loop=true
    UseLoopDelay=false
   )
   &self:InterpolateSpeedThresholdEmitter_04(
    Define Initial Value=false
    Rate Of Change=4
    Target Float Value:NiagaraFloat=&MakeCustomFloatFromBool(
     Bool=&SetBoolByFloatComparison(
      A=&Add_Float(
       A=&Add_Float(
        A=&Abs_Float(Float:NiagaraFloat=&MakeFloatFromVector(Vector=$Engine.Owner.Velocity, Channel=::NewEnumerator0))
        B=&Abs_Float(Float:NiagaraFloat=&MakeFloatFromVector(Vector=$Engine.Owner.Velocity, Channel=::NewEnumerator2))
       )
       B=&Abs_Float(Float:NiagaraFloat=&MakeFloatFromVector(Vector=$Engine.Owner.Velocity, Channel=::NewEnumerator1))
      )
      B=300
     )
     True Float=1
     False Float=0
    )
   ) {module=InterpolateSpeedThresholdEmitter_04}
   &self:InterpolateSpeedThresholdEmitter_04(
    Define Initial Value=false
    Rate Of Change=5
    Target Float Value:NiagaraFloat=&MakeCustomFloatFromBool(
     Bool=&SetBoolByFloatComparison(
      A=&Add_Float(
       A=&Add_Float(
        A=&Abs_Float(Float:NiagaraFloat=&MakeFloatFromVector(Vector=$Engine.Owner.Velocity, Channel=::NewEnumerator0))
        B=&Abs_Float(Float:NiagaraFloat=&MakeFloatFromVector(Vector=$Engine.Owner.Velocity, Channel=::NewEnumerator2))
       )
       B=&Abs_Float(Float:NiagaraFloat=&MakeFloatFromVector(Vector=$Engine.Owner.Velocity, Channel=::NewEnumerator1))
      )
      B=700
     )
     True Float=1
     False Float=0
    )
   ) {module=InterpolateSpeedThresholdEmitter_04001}
   &SpawnRate(
    SpawnRate=&Lerp_Float(A=32, B=32, Alpha=$Emitter.InterpolateSpeedThresholdEmitter_04.InterpolateOverTime_Float.Moving Average)
    Use Spawn Probability=false
   )
   /SetVariables(Emitter.float:NiagaraFloat=&FloatFromCurve(
     FloatCurve=DataInterface() {propertyValues {Curve {keys [<at 0 0>,<at 0.2775 1>,<at 0.551 0>,<at 1.1027 0>,<at 1.2839 1>,<at 1.6886 0>,<at 2.2688 0>,<at 2.3967 1>,<at 2.9 0>,<at 3.4741 0>,<at 3.5614 0.95>,<at 3.9 0>,<at 4.7292 0>,<at 4.8468 0.95>,<at 5.1784 0>,<at 5.7 0>,<at 5.9208 1>,<at 6.0989 0>,<at 6.84 0>,<at 7.151 1>,<at 7.4261 0>,<at 8.1 0>,<at 8.3099 1>,<at 9.0889 0>]}}}
     CurveIndex=$Emitter.LoopedAge
    )) {isSetParameters=true}
   /SetVariables(Emitter.float:NiagaraFloat=&FloatFromCurve(
     FloatCurve=DataInterface() {propertyValues {Curve {keys [<at 0 0>,<at 0.2 1>,<at 0.52 0>,<at 1 0>,<at 1.3 1>,<at 1.7 0>,<at 2.1 0>,<at 2.3 1>,<at 2.7 0>,<at 3.2 0>,<at 3.4 0.95>,<at 3.7 0>,<at 4.5 0>,<at 4.7 0.95>,<at 5 0>,<at 5.6 0>,<at 5.8 1>,<at 6.1 0>,<at 6.7 0>,<at 6.9 1>,<at 7.2 0>,<at 7.9 0>,<at 8.1 1>,<at 8.5 0>]}}}
     CurveIndex=$Emitter.LoopedAge
    )) {enabled=false, isSetParameters=true}
  ]
  particleSpawnScript [
   &InitializeParticle(
    Color=<rgba 0.6 0.2946 0 1>
    Lifetime=&Lerp_Float(A=0.65, B=0.6, Alpha=$Emitter.InterpolateSpeedThresholdEmitter_04.InterpolateOverTime_Float.Moving Average)
    Mesh Scale Mode=::NewEnumerator3
    Position Offset=<0 1 0>
    Ribbon Width=&Lerp_Float(A=22, B=22, Alpha=$Emitter.InterpolateSpeedThresholdEmitter_04.InterpolateOverTime_Float.Moving Average)
    Ribbon Width Mode=::NewEnumerator1
    UsePositionOffset=false
    Write Material Random=true
   )
   &SkeletalMeshLocation(
    MaskSpawnGroup=false
    Mesh Sampling Type=::NewEnumerator5
    Orientation Sampling=::NewEnumerator1
    Skeletal Mesh=$System.Skeletal Mesh
    Socket Sampling Mode=::NewEnumerator5
   ) {enabled=false, module=SkeletalMeshLocation002}
   &SkeletalMeshLocation(
    Bone Sampling Mode=::NewEnumerator4
    MaskSpawnGroup=false
    Orientation Sampling=::NewEnumerator1
    Position Sampling=::NewEnumerator2
    Sampled Position Offset=&MakeVector(
     X=&Waveform(
      Global Phase=$Engine.System.Age
      "Waveform [1]"=::NewEnumerator0
      "[1] Phase"=$Engine.Time
      "[1] Frequency"=2
      "[1] Amplitude Min/Max"=&Lerp_Vector2(A=<-15 15>, B=<-30 30>, Alpha=$Emitter.InterpolateSpeedThresholdEmitter_04.InterpolateOverTime_Float.Moving Average)
     )
     Y=&Add_Float(
      A=&Lerp_Float(A=-95, B=-140, Alpha=$Emitter.float)
      B=&Waveform(
       Waveform Count=::NewEnumerator1
       Global Input Position=0.12
       Global Phase=$Engine.System.Age
       "Waveform [1]"=::NewEnumerator0
       "[1] Phase"=$Engine.Time
       "[1] Frequency"=2
       "[1] Amplitude Min/Max"=&Lerp_Vector2(A=<-15 0>, B=<-10 0>, Alpha=$Emitter.float)
       "Waveform [2] Blend"=::NewEnumerator5
       "[2] Interpolation"=0.5
       "[2] Invert Interpolation"=false
       "[2] Input Position"=0
       "[2] Phase"=$Emitter.Age
       "[2] Phase Scale"=1
       "[2] Compound Frequency"=2
       "[2] Frequency"=1
       "[2] Amplitude Min/Max"=&Lerp_Vector2(A=<-10 0>, B=<-10 0>, Alpha=$Emitter.float)
       "[2] Scale"=1
       "[2] Bias"=0
       "[2] Exponent"=1
       "[2] Pulse Duration"=0.1
       "[2] Clip Triangle Min/Max"=<0 0>
       "[2] Random Seed"=0
      )
     )
     Z=0
    )
    Sampled Position Offset Coordinate Space:ENiagaraSkelMeshTransforms=::NewEnumerator2
    Skeletal Mesh=$User.PlayerPawn
   )
   &SkeletalMeshLocation(
    Bone / Socket Index=1
    Bone Sampling Mode=::NewEnumerator4
    MaskSpawnGroup=false
    Orientation Sampling=::NewEnumerator1
    Position Sampling=::NewEnumerator2
    Sampled Position Offset=&MakeVector(X=0, Y=0, Z=0)
    Sampled Position Offset Coordinate Space:ENiagaraSkelMeshTransforms=::NewEnumerator2
    Skeletal Mesh=$System.Skeletal Mesh
   ) {module=SkeletalMeshLocation001}
  ]
  particleUpdateScript [
   &ParticleState()
   &SampleSkeletalMesh(
    AdditionalYawPitchRoll=false
    Bone Sampling Mode:ENiagaraBoneSamplingMode=::NewEnumerator4
    Skeletal Mesh:NiagaraDataInterfaceSkeletalMesh=$System.Skeletal Mesh
   ) {module=SampleSkeletalMesh001}
   &SkeletalMeshLocation(
    Bone / Socket Index=2
    MaskSpawnGroup=false
    Mesh Sampling Type=::NewEnumerator5
    Orientation Sampling=::NewEnumerator1
    Position Sampling=::NewEnumerator2
    Sampled Position Offset=&MakeVector(X=0, Y=0, Z=0)
    Skeletal Mesh=$System.Skeletal Mesh
    Socket Sampling Mode=::NewEnumerator5
   ) {module=SkeletalMeshLocation003}
   &SkeletalMeshLocation(
    Bone / Socket Index=3
    MaskSpawnGroup=false
    Mesh Sampling Type=::NewEnumerator5
    Orientation Sampling=::NewEnumerator1
    Position Sampling=::NewEnumerator2
    Sampled Position Offset=&MakeVector(X=0, Y=0, Z=0)
    Skeletal Mesh=$System.Skeletal Mesh
    Socket Sampling Mode=::NewEnumerator5
   ) {module=SkeletalMeshLocation004}
   &ScaleRibbonWidth(Ribbon Width Scale=&Lerp_Float(
     A=0.1
     B=0.3
     Alpha=&FloatFromCurve(FloatCurve=DataInterface() {propertyValues {Curve {
         defaultValue=1
         keys [
          <at 0 0.3 RCIM_Cubic>
          {interpMode=RCIM_Cubic, tangentMode=RCTM_User, time=0.15, value=1}
          <at 0.3 2 RCIM_Cubic>
          {arriveTangent=-2.8571, interpMode=RCIM_Cubic, leaveTangent=-2.8571, time=0.6, value=1}
          {arriveTangent=-6.4418, tangentMode=RCTM_User, time=1, value=0}
         ]
        }}})
    ))
   &ScaleColor(
    Scale Alpha=&Lerp_Float(
     A=0
     B=1
     Alpha=&FloatFromCurve(FloatCurve=DataInterface() {propertyValues {Curve {keys [
          {arriveTangent=-3.2009, interpMode=RCIM_Cubic, leaveTangent=-3.2009, tangentMode=RCTM_User, time=0.6, value=1}
          {arriveTangent=-0.0448, interpMode=RCIM_Cubic, leaveTangent=-0.0448, tangentMode=RCTM_User, time=1, value=0.6}
         ]}}})
    )
    ScaleA=true
    ScaleRGB=true
   )
   /SetVariables(Particles.DistanceParticleToPlayer:NiagaraFloat=&DistanceBetweenPositions(A=$Particles.Position, B=/Niagara/DynamicInputs/Helpers/SimulationPosition())) {isSetParameters=true}
   /SetVariables(Particles.RibbonWidth=&Multiply_Float(
     A=$Particles.RibbonWidth
     B=&MakeCustomFloatFromBool(
      Bool=&SetBoolByFloatComparison(
       A=$Particles.DistanceParticleToPlayer
       B=&Lerp_Float(A=200, B=400, Alpha=$Emitter.InterpolateSpeedThresholdEmitter_04.InterpolateOverTime_Float.Moving Average)
      )
      True Float=0
      False Float=1
     )
    )) {isSetParameters=true}
   /SetVariables(Particles.GrindTrainSpeedPos:NiagaraPosition=&AddVectorToPosition(Position=$Particles.SkeletalMeshLocation004.SampledPosition, Vector=&MakeVector(X=0, Y=0, Z=0))) {isSetParameters=true}
   &self:ScratchModule_09(
    IsFrontEnd:NiagaraBool=$User.bIsFrontEndPreview
    PlayerPos:NiagaraPosition=$Engine.Emitter.SimulationPosition
    TraceDir:Vector3f=<0 0 1>
    TraceDist=300
   ) {enabled=false, module=ScratchModule_09}
   &self:ScratchModule_08(
    CeilingReduction=1
    DebugDraw001=DataInterface() {propertyValues {OverrideMaxLineInstances=5}}
    MustardLandingPos:NiagaraPosition=$Particles.SkeletalMeshLocation.SampledPosition
    TraceDir:Vector3f=<0 0 -1>
    TraceDist=120
    TraceVertical=85
    VerticalAdjustment=4
    bIsFrontEnd:NiagaraBool=$User.bIsFrontEndPreview
   ) {module=ScratchModule_08}
   /SetVariables(Particles.RibbonWidth=&Multiply_Float(
     A=$Particles.RibbonWidth
     B=&Lerp_Float(
      A=&FloatFromCurve(FloatCurve=DataInterface() {propertyValues {Curve {keys [<at 0 1>,<at 0.5 0>]}}})
      B=1
      Alpha=&MakeCustomFloatFromBool(Bool=$Particles.CollisionValid, True Float=1, False Float=0)
     )
    )) {isSetParameters=true}
   &KillParticles(Kill Particles=&SetBoolByFloatComparison(A=$Particles.RibbonWidth, B=0, Comparison Type=::NewEnumerator7))
   /SetVariables(
    Particles.NozzleLoc:NiagaraPosition=$Particles.SampleSkeletalMesh001.SampledPosition
    Particles.Position=&Lerp_Vector(
     A=&Lerp_Vector(
      A=&Lerp_Vector(
       A=$Particles.SampleSkeletalMesh001.SampledPosition
       B=$Particles.SkeletalMeshLocation003.SampledPosition
       Alpha=&FloatFromCurve(FloatCurve=DataInterface() {propertyValues {Curve {defaultValue=1, keys [{tangentMode=RCTM_User, time=0, value=0}, {tangentMode=RCTM_User, time=0.3, value=1}]}}})
      )
      B=Linked() {linkedVariable {name=Particles.CollisionPos, type=NiagaraPosition}}
      Alpha=&FloatFromCurve(FloatCurve=DataInterface() {propertyValues {Curve {defaultValue=1, keys [{tangentMode=RCTM_User, time=0, value=0}, {tangentMode=RCTM_User, time=0.3, value=1}]}}})
     )
     B=&Lerp_Vector(
      A=$Particles.SampleSkeletalMesh001.SampledPosition
      B=Linked() {linkedVariable {name=Particles.GrindTrainSpeedPos, type=NiagaraPosition}}
      Alpha=&FloatFromCurve(FloatCurve=DataInterface() {propertyValues {Curve {defaultValue=1, keys [{tangentMode=RCTM_User, time=0, value=0}, {tangentMode=RCTM_User, time=1, value=1}]}}})
     )
     Alpha=$Emitter.InterpolateSpeedThresholdEmitter_04001.InterpolateOverTime_Float.Moving Average
    )
   ) {enabled=false, isSetParameters=true}
   /SetVariables(
    Particles.NozzleLoc:Vector3f=$Particles.SampleSkeletalMesh001.SampledPosition
    Particles.Position=&Lerp_Vector(
     A=&Lerp_Vector(
      A=&Lerp_Vector(
       A=$Particles.SampleSkeletalMesh001.SampledPosition
       B=$Particles.SkeletalMeshLocation003.SampledPosition
       Alpha=&FloatFromCurve(FloatCurve=DataInterface() {propertyValues {Curve {defaultValue=1, keys [{tangentMode=RCTM_User, time=0, value=0}, {tangentMode=RCTM_User, time=0.15, value=1}]}}})
      )
      B=Linked() {linkedVariable {name=Particles.CollisionPos, type=NiagaraPosition}}
      Alpha=&FloatFromCurve(FloatCurve=DataInterface() {propertyValues {Curve {defaultValue=1, keys [{tangentMode=RCTM_User, time=0, value=0}, {tangentMode=RCTM_User, time=0.15, value=1}]}}})
     )
     B=&Lerp_Vector(
      A=$Particles.SampleSkeletalMesh001.SampledPosition
      B=Linked() {linkedVariable {name=Particles.GrindTrainSpeedPos, type=NiagaraPosition}}
      Alpha=&FloatFromCurve(FloatCurve=DataInterface() {propertyValues {Curve {defaultValue=1, keys [{tangentMode=RCTM_User, time=0, value=0}, {tangentMode=RCTM_User, time=0.5, value=1}]}}})
     )
     Alpha=$Emitter.InterpolateSpeedThresholdEmitter_04001.InterpolateOverTime_Float.Moving Average
    )
   ) {isSetParameters=true}
   /SetVariables(Particles.Position=&AddVectorToPosition(
     Position=$Particles.Position
     Vector=&Lerp_Vector(
      A=<0 0 0>
      B=&MakeVector(
       X=&Waveform(
        "Waveform [1]"=::NewEnumerator0
        "[1] Input Position"=&ReturnNormalizedExecIndex()
        "[1] Frequency"=3
        "[1] Amplitude Min/Max"=<-8 8>
        "[1] Scale"=&FloatFromCurve(FloatCurve=DataInterface() {propertyValues {Curve {defaultValue=1, keys [{tangentMode=RCTM_User, time=0, value=0}, {tangentMode=RCTM_User, time=1, value=1}]}}})
        "Waveform [2] Blend"=::NewEnumerator5
       )
       Y=&Waveform(
        "Waveform [1]"=::NewEnumerator0
        "[1] Input Position"=&ReturnNormalizedExecIndex()
        "[1] Frequency"=3
        "[1] Amplitude Min/Max"=<-8 8>
        "[1] Scale"=&FloatFromCurve(FloatCurve=DataInterface() {propertyValues {Curve {defaultValue=1, keys [{tangentMode=RCTM_User, time=0, value=0}, {tangentMode=RCTM_User, time=1, value=1}]}}})
        "Waveform [2] Blend"=::NewEnumerator5
       )
       Z=&Waveform(
        "Waveform [1]"=::NewEnumerator0
        "[1] Input Position"=&ReturnNormalizedExecIndex()
        "[1] Frequency"=3
        "[1] Amplitude Min/Max"=<-8 8>
        "[1] Scale"=&FloatFromCurve(FloatCurve=DataInterface() {propertyValues {Curve {defaultValue=1, keys [{tangentMode=RCTM_User, time=0, value=0}, {tangentMode=RCTM_User, time=1, value=1}]}}})
        "Waveform [2] Blend"=::NewEnumerator5
       )
      )
      Alpha=$Emitter.InterpolateSpeedThresholdEmitter_04001.InterpolateOverTime_Float.Moving Average
     )
    )) {isSetParameters=true}
   /SetVariables(Particles.Lifetime=&Add_Float(
     A=$Particles.Lifetime
     B=&MakeCustomFloatFromBool(
      Bool=$Particles.MissedTrace
      True Float=&FloatFromCurve(
       FloatCurve=DataInterface() {propertyValues {Curve {defaultValue=1, keys [{tangentMode=RCTM_User, time=0, value=0}, {tangentMode=RCTM_User, time=1, value=1}]}}}
       Scale Curve=-0.2
      )
      False Float=0
     )
    )) {enabled=false, isSetParameters=true}
   &self:ScratchModule_07(NozzlePos:NiagaraPosition=$Particles.SampleSkeletalMesh001.SampledPosition) {module=ScratchModule_07}
   &DynamicMaterialParameters(
    Index 0 Param 1=&Lerp_Float(
     A=0
     B=1
     Alpha=&FloatFromCurve(FloatCurve=DataInterface() {propertyValues {Curve {keys [
          {arriveTangent=5.87, interpMode=RCIM_Cubic, leaveTangent=5.87, tangentMode=RCTM_User, time=0, value=0}
          {interpMode=RCIM_Cubic, tangentMode=RCTM_User, time=0.5, value=1}
         ]}}})
    )
    Index 0 Param 2=&Lerp_Float(
     A=0
     B=10
     Alpha=&FloatFromCurve(FloatCurve=DataInterface() {propertyValues {Curve {keys [
          {arriveTangent=5.87, interpMode=RCIM_Cubic, leaveTangent=5.87, tangentMode=RCTM_User, time=0, value=0}
          {interpMode=RCIM_Cubic, tangentMode=RCTM_User, time=0.5, value=1}
         ]}}})
    )
    Index 0 Param 3=&RandomRangeFloat(Minimum=0, Maximum=0)
    Index 0 Param 4=&RandomRangeFloat(Minimum=0, Maximum=0)
    Module.Index0Param0Write=true
    Module.Index0Param1Write=true
    Module.Index0Param2Write=true
    Module.Index0Param3Write=true
   )
  ]
  rendererClasses [&NiagaraRibbonRendererProperties]
  renderers [{rendererClass=&NiagaraRibbonRendererProperties}]
 }
}
summary {
 systemName=NS_Emote_MustardRocket_Spray
 emitters [
  {emitterName=MustardMuzzle, rendererClasses [&NiagaraMeshRendererProperties]}
  {emitterName=MustardTrail, rendererClasses [&NiagaraRibbonRendererProperties]}
  {emitterName=MustardLiquid, rendererClasses [&NiagaraSpriteRendererProperties]}
  {emitterName=MustardDecals, rendererClasses [&NiagaraDecalRendererProperties]}
 ]
 userVariables [
  User.bIsFrontEnd {type=NiagaraBool}
  User.bIsFrontEndPreview {type=NiagaraBool}
  User.ColorB {type=LinearColor, description="", defaultValue {struct=LinearColor, value=<rgba 1 1 1 1>}}
  User.ColorC {type=LinearColor, description="", defaultValue {struct=LinearColor, value=<rgba 1 1 1 1>}}
  User.PlayerPawn {
   type=NiagaraDataInterfaceSkeletalMesh
   defaultValue {
    struct=NiagaraExt_VariableValue_DataInterface
    value {dataInterface=&self:NiagaraDataInterfaceSkeletalMesh_0}
   }
  }
  User.Rotation {type=NiagaraFloat, defaultValue=0, description=""}
  User.Squeeze {type=NiagaraFloat, defaultValue=0, description=""}
 ]
}
